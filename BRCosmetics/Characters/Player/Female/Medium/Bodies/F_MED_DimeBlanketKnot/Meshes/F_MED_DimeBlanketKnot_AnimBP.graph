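# Post-process style animation Blueprint for the F_MED_DimeBlanketKnot female cosmetic body, which follows its attached parent mesh.
# The copied parent pose runs through the body's Control Rig, then a per-bone layered blend mixes DefaultSlot montages onto both arms.
# In the front end, the update event sets that blend weight from the owner anim instance's Frontend_Idle_Bypass curve.
# The left-hand branch filter names "Nonehand_l" rather than hand_l.
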AnimBlueprint /BRCosmetics/Characters/Player/Female/Medium/Bodies/F_MED_DimeBlanketKnot/Meshes/F_MED_DimeBlanketKnot_AnimBP
target_skeleton=&/F_MED_DimeBlanketKnot_Skeleton
graphs [
 AnimGraph {nodes [
   n1 = Control Rig(Source=Copy Pose From Mesh() {bUseAttachedParent=true}) {
    alphaBoolBlend {blendOption=Linear}
    controlRigAssetReference {blueprintRigClass=&/F_MED_DimeBlanketKnot_CtrlRig.F_MED_DimeBlanketKnot_CtrlRig_C}
   }
   n2 = Layered blend per bone(BasePose=@n1, BlendPoses_0=Slot 'DefaultSlot'(), BlendWeights_0=Get Custom Idle is Playing Float() -> (Custom Idle is Playing Float:double)) {layerSetup [{branchFilters [
       {boneName=upperarm_r}
       {boneName=upperarm_twist_01_r}
       {boneName=upperarm_twist_02_r}
       {boneName=lowerarm_r}
       {boneName=lowerarm_twist_01_r}
       {boneName=lowerarm_twist_02_r}
       {boneName=hand_r}
       {boneName=upperarm_l}
       {boneName=upperarm_twist_01_l}
       {boneName=upperarm_twist_02_l}
       {boneName=lowerarm_l}
       {boneName=lowerarm_twist_01_l}
       {boneName=lowerarm_twist_02_l}
       {boneName=Nonehand_l}
      ]}]}
   Output Pose(Result=@n2) {layerGroup=None}
  ]}
 EventGraph {nodes [
   Event BlueprintUpdateAnimation()
   > Branch(Condition=Get bIsInFrontEnd())
   > Set Custom Idle is Playing Float(Custom Idle is Playing Float:double=GetCurveValue(self=GetOwnerFrontEndAnimInstance(), CurveName=Frontend_Idle_Bypass)) -> (Output_Get:double)
  ]}
]
variables [Custom Idle is Playing Float {type=number}]
